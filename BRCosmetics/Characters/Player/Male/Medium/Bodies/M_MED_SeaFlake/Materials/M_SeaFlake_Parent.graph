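# Parent character material for the SeaFlake medium male player body, wrapping skin-specific effects between the shared 2023 character PreFX and PostFX functions.
# By default it forces Subsurface Profile shading and blends opacity toward a scalable value using the Tech Art Mask blue channel.
# A static switch replaces that path with a Banana weapon-skin function driven by a green mask.
# A second switch layers sub-UV textures through a selectable mask channel.
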Material /BRCosmetics/Characters/Player/Male/Medium/Bodies/M_MED_SeaFlake/Materials/M_SeaFlake_Parent
graph {
 properties {
  bUseMaterialAttributes=true
  bUsedWithMeshDeformer=true
  bUsedWithMorphTargets=true
  bUsedWithSkeletalMesh=true
  pixelDepthOffsetMode=PDOM_Legacy
  shadingModel=MSM_FromMaterialExpression
  usedShadingModels=Default Lit | Subsurface | Clear Coat | Subsurface Profile | Cloth
 }
 nodes [
  n1 = MaterialFunctionCall() {materialFunction=&MF_CharacterParent_2023_PreFX}
  n2 = ChannelMaskParameter(None=@n1.Tech Art Mask) {parameterName=Sub UV Mask Channel}
  n3 = GetMaterialAttributes(None=@n1.Result) {attributeGetTypes [#1]}
  n4 = ComponentMask(None=@n1.Tech Art Mask) {b=true}
  n5 = Multiply(A=ScalarParameter() {defaultValue=1, group=DashSurge, parameterName=SubsurfaceOpacityMapMultiplier}, B=@n4)
  n6 = StaticSwitchParameter(
   True=MaterialFunctionCall!(Gmap=ComponentMask(None=@n1.M) {g=true}, Material) {materialFunction=/Game/Weapons/WeaponSkins/MaterialFunction/Banana}.Result
   False=SetMaterialAttributes!(
    MaterialAttributes=@n3.MaterialAttributes
    Shading Model=ShadingModel() {shadingModel=MSM_SubsurfaceProfile}
    Opacity=LinearInterpolate(A=@n3.Opacity, B=@n5, Alpha=@n4)
   ) {attributeSetTypes [#2, #1]}
  ) {group=Skin FX, parameterName=Use Banana}
  MaterialFunctionCall(
   Material=StaticSwitchParameter(
    True=MaterialFunctionCall!(Alpha=@n2.R, Material=@n6) {materialFunction=/Game/Weapons/WeaponSkins/MaterialFunction/SubUVTextures}.Result
    False=@n6
   ) {group=Skin FX, parameterName=Use Sub UV texture}
   M=@n1.M
   Pre Skinned Local Position=@n1.M
   Distance Blend=@n1.M
   Fresnel=@n1.M
  ) {materialFunction=&MF_CharacterParent_2023_PostFX}
 ]
}
