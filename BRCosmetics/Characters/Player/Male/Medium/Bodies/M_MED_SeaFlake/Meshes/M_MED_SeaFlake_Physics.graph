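# Physics asset for the SeaFlake medium male player body cosmetic, covering its jacket panels and a pouch.
# Tiny capsules on base bones anchor constrained chains of dyn_ capsules: two links per back, side and front panel, one for the pouch.
# Larger capsules on pelvis, spine_01 and the right thigh form body volumes without constraints; the left thigh has none.
# The first back link has all angular motion locked to its base.
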PhysicsAsset /BRCosmetics/Characters/Player/Male/Medium/Bodies/M_MED_SeaFlake/Meshes/M_MED_SeaFlake_Physics
bodies {
 base_jacket_back [{shapeType=Capsule, length=0.505, radius=0.505, shapeName=base_jacket_back_capsule, rotation=<rot 0 -90 0>}]
 base_jacket_front_l [{shapeType=Capsule, length=0.505, radius=0.505, shapeName=base_jacket_front_l_capsule, rotation=<rot 0 -90 0>}]
 base_jacket_front_r [{shapeType=Capsule, length=0.505, radius=0.505, shapeName=base_jacket_front_r_capsule, rotation=<rot 0 -90 0>}]
 base_jacket_side_l [{shapeType=Capsule, length=0.505, radius=0.505, shapeName=base_jacket_side_l_capsule, rotation=<rot 0 -90 0>}]
 base_jacket_side_r [{shapeType=Capsule, length=0.505, radius=0.505, shapeName=base_jacket_side_r_capsule, rotation=<rot 0 -90 0>}]
 base_pouch_01 [{shapeType=Capsule, length=0.505, radius=0.505, shapeName=base_pouch_01_capsule, rotation=<rot 0 -90 0>}]
 dyn_jacket_back_01 [{shapeType=Capsule, length=14.8, radius=4.3, shapeName=dyn_jacket_back_01_capsule, center=<6.3 0 4>, rotation=<rot 0 0 -88>}]
 dyn_jacket_back_02 [{shapeType=Capsule, length=20.1354, radius=3.7716, shapeName=dyn_jacket_back_02_capsule, center=<3.6 0 4.1>, rotation=<rot 0 -0.3 -88>}]
 dyn_jacket_front_01_l [{shapeType=Capsule, length=10, radius=3.5, shapeName=dyn_jacket_front_01_l_capsule, center=<6 0 3.5>, rotation=<rot 0 -178 118>}]
 dyn_jacket_front_01_r [{shapeType=Capsule, length=10.3, radius=4, shapeName=dyn_jacket_front_01_r_capsule, center=<5.8 0 1.8>, rotation=<rot -0.5 1.2 -62>}]
 dyn_jacket_front_02_l [{shapeType=Capsule, length=12.5, radius=4.2, shapeName=dyn_jacket_front_02_l_capsule, center=<5 0 3.5>, rotation=<rot 0 175 114>}]
 dyn_jacket_front_02_r [{shapeType=Capsule, length=12.5, radius=4.2, shapeName=dyn_jacket_front_02_r_capsule, center=<5.4 0 1.7>, rotation=<rot 0 0 -66.5>}]
 dyn_jacket_side_01_l [{shapeType=Capsule, length=18, radius=4.4, shapeName=dyn_jacket_side_01_l_capsule, center=<5.11 0 1.7>, rotation=<rot 0 -177 -80>}]
 dyn_jacket_side_01_r [{shapeType=Capsule, length=11.8, radius=5, shapeName=dyn_jacket_side_01_r_capsule, center=<5.7 0 2.11>, rotation=<rot -1.2 3.4 -82>}]
 dyn_jacket_side_02_l [{shapeType=Capsule, length=11.4, radius=4.5, shapeName=dyn_jacket_side_02_l_capsule, center=<4.8 0 2.4>, rotation=<rot -1.4 -176 -82>}]
 dyn_jacket_side_02_r [{shapeType=Capsule, length=11.4, radius=4.5, shapeName=dyn_jacket_side_02_r_capsule, center=<4.8 0 2>, rotation=<rot -1.4 3 -83>}]
 dyn_pouch_02 [{shapeType=Capsule, length=6, radius=5, shapeName=dyn_pouch_02_capsule, center=<6 0 4>, rotation=<rot 0 90 -90>}]
 pelvis [
  {
   shapeType=Capsule
   length=10.337
   radius=10.1088
   shapeName=pelvis_capsule
   center=<-0.1822 -1.1777 3.4085>
   rotation=<rot -0.2342 -3.815 0.0156>
  }
  {
   shapeType=Capsule
   length=1.3675
   radius=15
   shapeName=pelvis_capsule
   center=<6.832 -0.6277 -0.3616>
   rotation=<rot -3.0174 -5.2498 -0.3462>
  }
 ]
 spine_01 [
  {shapeType=Capsule, length=18.9002, radius=13.0998, shapeName=spine_01_capsule, center=<5.9454 0 0>, rotation=<rot 0 -90 90>}
  {shapeType=Capsule, length=5.8257, radius=11.0147, shapeName=spine_01_capsule, center=<2.0698 1.6164 0>}
  {shapeType=Capsule, length=10.4953, radius=9.3238, shapeName=spine_01_capsule, center=<3.5318 0 0.8421>}
 ]
 thigh_r [{
   shapeType=Capsule
   length=27.1837
   radius=9.1642
   shapeName=thigh_r_capsule
   center=<16.2396 -1.2884 0.8299>
   rotation=<rot 85.4731 179.955 -176.068>
  }]
}
constraints [
 {bone1Name=dyn_jacket_front_01_l, bone2Name=base_jacket_front_l, swing1LimitDegrees=5, swing2LimitDegrees=15, twistLimitDegrees=0}
 {bone1Name=dyn_jacket_side_01_l, bone2Name=base_jacket_side_l, swing1LimitDegrees=0, swing2LimitDegrees=2, twistLimitDegrees=0}
 {bone1Name=dyn_jacket_side_02_l, bone2Name=dyn_jacket_side_01_l, swing1LimitDegrees=0, swing2LimitDegrees=2, twistLimitDegrees=0}
 {
  bone1Name=dyn_jacket_back_01
  bone2Name=base_jacket_back
  swing1LimitDegrees=0
  swing1Motion=Locked
  swing2LimitDegrees=2
  swing2Motion=Locked
  twistLimitDegrees=0
  twistMotion=Locked
 }
 {bone1Name=dyn_jacket_back_02, bone2Name=dyn_jacket_back_01, swing1LimitDegrees=0, swing2LimitDegrees=2, twistLimitDegrees=0}
 {bone1Name=dyn_jacket_side_01_r, bone2Name=base_jacket_side_r, swing1LimitDegrees=0, swing2LimitDegrees=2, twistLimitDegrees=0}
 {bone1Name=dyn_jacket_side_02_r, bone2Name=dyn_jacket_side_01_r, swing1LimitDegrees=0, swing2LimitDegrees=2, twistLimitDegrees=0}
 {bone1Name=dyn_jacket_front_01_r, bone2Name=base_jacket_front_r, swing1LimitDegrees=5, swing2LimitDegrees=15, twistLimitDegrees=0}
 {bone1Name=dyn_jacket_front_02_r, bone2Name=dyn_jacket_front_01_r, swing1LimitDegrees=5, swing2LimitDegrees=15, twistLimitDegrees=0}
 {bone1Name=dyn_jacket_front_02_l, bone2Name=dyn_jacket_front_01_l, swing1LimitDegrees=5, swing2LimitDegrees=15, twistLimitDegrees=0}
 {bone1Name=dyn_pouch_02, bone2Name=base_pouch_01, swing1LimitDegrees=5, swing2LimitDegrees=15, twistLimitDegrees=0}
]
