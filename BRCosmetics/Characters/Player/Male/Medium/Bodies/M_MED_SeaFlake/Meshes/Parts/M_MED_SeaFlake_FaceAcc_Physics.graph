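# Physics asset for the M_MED_SeaFlake male character cosmetic's face accessory, defining capsule bodies and joints for left and right seaweed strands.
# Each strand is a three-link dyn_seaweed chain constrained to its small base_seaweed body, with very tight swing limits and zero twist limit.
# Unconstrained capsules on the head, both clavicles and only the right upper arm provide body shapes around the strands.
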PhysicsAsset /BRCosmetics/Characters/Player/Male/Medium/Bodies/M_MED_SeaFlake/Meshes/Parts/M_MED_SeaFlake_FaceAcc_Physics
bodies {
 base_seaweed_l [{shapeType=Capsule, length=0.505, radius=0.505, shapeName=base_seaweed_l_capsule, rotation=<rot 0 -90 0>}]
 base_seaweed_r [{shapeType=Capsule, length=0.505, radius=0.505, shapeName=base_seaweed_r_capsule, rotation=<rot 0 -90 0>}]
 clavicle_l [{shapeType=Capsule, length=10, radius=5, shapeName=clavicle_l_capsule, center=<8 0 0>, rotation=<rot 0 -90 90>}]
 clavicle_r [{shapeType=Capsule, length=10, radius=5, shapeName=clavicle_r_capsule, center=<-8 0 0>, rotation=<rot 0 -90 90>}]
 dyn_seaweed_01_l [{shapeType=Capsule, length=5, radius=2, shapeName=dyn_seaweed_01_l_capsule, rotation=<rot 90 -90 0>}]
 dyn_seaweed_01_r [{shapeType=Capsule, length=5.2, radius=2.8, shapeName=dyn_seaweed_01_r_capsule, center=<1.2 0 1>, rotation=<rot 72 86.5 0>}]
 dyn_seaweed_02_l [{shapeType=Capsule, length=5, radius=3, shapeName=dyn_seaweed_02_l_capsule, center=<2 0 0>, rotation=<rot 90 0 -10>}]
 dyn_seaweed_02_r [{shapeType=Capsule, length=5, radius=3, shapeName=dyn_seaweed_02_r_capsule, center=<2 0 0>, rotation=<rot 90 0 10>}]
 dyn_seaweed_03_l [{shapeType=Capsule, length=5, radius=3, shapeName=dyn_seaweed_03_l_capsule, center=<2 0 0>, rotation=<rot 90 0 -40>}]
 dyn_seaweed_03_r [{shapeType=Capsule, length=2, radius=3, shapeName=dyn_seaweed_03_r_capsule, center=<2 0 0>, rotation=<rot 90 0 20>}]
 head [
  {shapeType=Capsule, length=5.9058, radius=10, center=<4.5428 6 0>}
  {shapeType=Capsule, length=13.189, radius=8.561, shapeName=head_capsule, center=<10.7586 6.7127 -0.3021>}
 ]
 upperarm_r [{shapeType=Capsule, length=15, radius=5, shapeName=upperarm_r_capsule, center=<-10 0 0>, rotation=<rot 0 -90 90>}]
}
constraints [
 {bone1Name=dyn_seaweed_01_l, bone2Name=base_seaweed_l, swing1LimitDegrees=1.5, swing2LimitDegrees=0.5, twistLimitDegrees=0}
 {bone1Name=dyn_seaweed_02_l, bone2Name=dyn_seaweed_01_l, swing1LimitDegrees=1.5, swing2LimitDegrees=0.5, twistLimitDegrees=0}
 {bone1Name=dyn_seaweed_03_l, bone2Name=dyn_seaweed_02_l, swing1LimitDegrees=1.5, swing2LimitDegrees=0.5, twistLimitDegrees=0}
 {bone1Name=dyn_seaweed_01_r, bone2Name=base_seaweed_r, swing1LimitDegrees=1.5, swing2LimitDegrees=0.5, twistLimitDegrees=0}
 {bone1Name=dyn_seaweed_02_r, bone2Name=dyn_seaweed_01_r, swing1LimitDegrees=1.5, swing2LimitDegrees=0.5, twistLimitDegrees=0}
 {bone1Name=dyn_seaweed_03_r, bone2Name=dyn_seaweed_02_r, swing1LimitDegrees=1.5, swing2LimitDegrees=0.5, twistLimitDegrees=0}
]
