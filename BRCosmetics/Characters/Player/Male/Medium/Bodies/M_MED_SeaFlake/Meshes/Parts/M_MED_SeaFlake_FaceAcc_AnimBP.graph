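# Post-attach animation Blueprint for the SeaFlake medium male skin's face accessory, adding secondary physics motion to its dangling parts.
# It copies the pose from the attached parent mesh, then runs a RigidBody simulation in base-bone space using the accessory's own physics asset.
# Two chained AnimDynamics nodes then add sprung, angle-limited sway with reduced gravity to the dyn_Sushi_01 and dyn_Sushi_02 bones.
# The event graph holds no logic.
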AnimBlueprint /BRCosmetics/Characters/Player/Male/Medium/Bodies/M_MED_SeaFlake/Meshes/Parts/M_MED_SeaFlake_FaceAcc_AnimBP
target_skeleton=&/M_MED_SeaFlake_FaceAcc_Skeleton
graphs [
 AnimGraph {nodes [
   n1 = RigidBody(ComponentPose=Local To Component(LocalPose=Copy Pose From Mesh() {bUseAttachedParent=true})) {
    overridePhysicsAsset=&/M_MED_SeaFlake_FaceAcc_Physics
    simulationSpace=BaseBoneSpace
    alphaBoolBlend {blendOption=Linear}
    simSpaceSettings {maxAngularVelocity=5000, maxLinearVelocity=3500, worldAlpha=0.2}
   }
   n2 = Anim Dynamics - Bone: dyn_Sushi_01(ComponentPose=@n1, Alpha=1, ExternalForce="0,0,0") {
    angularDampingOverride=1
    angularSpringConstant=20000
    bAngularSpring=true
    bOverrideAngularDamping=true
    bUseGravityOverride=true
    gravityScale=0.35
    simulationSpace=RootRelative
    alphaBoolBlend {blendOption=Linear}
    boundBone {boneName=dyn_Sushi_01}
    componentLinearAccScale=<1 1 1>
    componentLinearVelScale=<1 1 1>
    gravityOverride {z=-980}
    physicsBodyDefinitions [{
      boundBone {boneName=dyn_Sushi_01}
      boxExtents=<20 30 30>
      constraintSetup {angularLimitsMax=<5 0 5>, angularLimitsMin=<-5 -60 -5>, angularTarget=<1 0 0>}
      localJointOffset=<4 0 0>
     }]
   }
   n3 = Anim Dynamics - Bone: dyn_Sushi_02(ComponentPose=@n2, Alpha=1, ExternalForce="0,0,0") {
    angularDampingOverride=1
    angularSpringConstant=20000
    bAngularSpring=true
    bOverrideAngularDamping=true
    bUseGravityOverride=true
    gravityScale=0.35
    simulationSpace=RootRelative
    alphaBoolBlend {blendOption=Linear}
    boundBone {boneName=dyn_Sushi_02}
    componentLinearAccScale=<1 1 1>
    componentLinearVelScale=<1 1 1>
    gravityOverride {z=-980}
    physicsBodyDefinitions [{
      boundBone {boneName=dyn_Sushi_02}
      boxExtents=<20 30 30>
      constraintSetup {angularLimitsMax=<10 15 0>, angularLimitsMin=<-10 -20 0>, angularTarget=<1 0 0>}
      localJointOffset=<4 0 0>
     }]
   }
   Output Pose(Result=Component To Local(ComponentPose=@n3)) {layerGroup=None}
  ]}
 EventGraph {nodes [Event BlueprintUpdateAnimation()]}
]
variables []
